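# Material function that samples a texture with Catmull-Rom spline filtering, described as a costly brute-force method to use sparingly.
# Tiled UVs are split into texel center and fraction, assuming a hard-coded 512 size, to take a 4x4 grid of BrushRT red-channel samples at MipLevel.
# Four GenerateASpline calls interpolate the rows along X, and a fifth blends them along Y into the output.
# A Texture2DSampleBicubic custom node is present but unconnected.
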MaterialFunction /Water/Materials/Functions/CatmullRomTexture
graph {
 properties {description=This is an extremely costly brute force solution to sample a texture with Catmul-Rom spline filtering. It should be used sparingly.}
 nodes [
  MaterialExpressionTextureObject_0 = TextureObject() {texture=/Engine/MapTemplates/TilingNoise05}
  MaterialExpressionMultiply_0 = Multiply(
   A=TextureCoordinate()
   B=FunctionInput(Preview=ScalarParameter() {defaultValue=2, parameterName=T}) {bUsePreviewValueAsDefault=true, inputName=Tiling, inputType=FunctionInput_Scalar}
  )
  MaterialExpressionConstant_0 = Constant() {r=512}
  Custom!(Tex=@MaterialExpressionTextureObject_0, UV=@MaterialExpressionMultiply_0, Size=@MaterialExpressionConstant_0) {
   code="return Texture2DSampleBicubic(Tex, TexSampler, UV, Size, 1/Size);"
   outputType=CMOT_Float1
   inputs [
    {inputName=Tex, input {expression=&self:MaterialExpressionTextureObject_0}}
    {inputName=UV, input {expression=&self:MaterialExpressionMultiply_0}}
    {inputName=Size, input {expression=&self:MaterialExpressionConstant_0}}
   ]
  }
  n1 = Add(A=Multiply(A=@MaterialExpressionMultiply_0) {constB=512, desc=* TextureSize}) {constB=-0.5}
  n2 = Frac(None=@n1)
  n3 = ComponentMask(None=Frac(None=@n1)) {r=true}
  n4 = Add(A=Floor(None=@n1)) {constB=0.5}
  n5 = Divide() {constB=512, desc=Texel Size}
  n6 = Subtract(A=Multiply(A=Divide(A=@n4) {constB=512, desc=/ TextureSize}) {desc=Current Texel Center}, B=Multiply(A=@n5, B=Constant2Vector() {g=1}))
  n7 = Multiply(A=@n5, B=Constant2Vector() {r=1})
  n8 = ScalarParameter() {parameterName=MipLevel}
  n9 = TextureSampleParameter2D!(UVs=Subtract(A=@n6, B=@n7), Level=@n8, Apply View MipBias) {desc=-X Texel, mipValueMode=TMVM_MipLevel, parameterName=BrushRT, texture=&TilingNoise05}
  n10 = TextureSampleParameter2D(UVs=@n6, Level=@n8) {desc="Current Texel (upper left)", mipValueMode=TMVM_MipLevel, parameterName=BrushRT, texture=&TilingNoise05}
  n11 = Add(A=@n7, B=@n6)
  n12 = TextureSampleParameter2D(UVs=@n11, Level=@n8) {desc=+X Texel, mipValueMode=TMVM_MipLevel, parameterName=BrushRT, texture=&TilingNoise05}
  n13 = TextureSampleParameter2D(UVs=Add(A=@n11, B=@n7), Level=@n8) {desc=++X Texel, mipValueMode=TMVM_MipLevel, parameterName=BrushRT, texture=&TilingNoise05}
  n14 = ComponentMask(None=@n2) {r=true}
  n15 = Floor(None=@n1)
  n16 = Multiply(A=Divide(A=Add(A=@n15) {constB=0.5}) {constB=512, desc=/ TextureSize}) {desc=Current Texel Center}
  n17 = Multiply(A=Divide() {constB=512, desc=Texel Size}, B=Constant2Vector() {r=1})
  n18 = TextureSampleParameter2D(UVs=Subtract(A=@n16, B=@n17), Level=@n8) {desc=-X Texel, mipValueMode=TMVM_MipLevel, parameterName=BrushRT, texture=&TilingNoise05}
  n19 = TextureSampleParameter2D(UVs=@n16, Level=@n8) {desc="Current Texel (upper left)", mipValueMode=TMVM_MipLevel, parameterName=BrushRT, texture=&TilingNoise05}
  n20 = Add(A=@n17, B=@n16)
  n21 = TextureSampleParameter2D(UVs=@n20, Level=@n8) {desc=+X Texel, mipValueMode=TMVM_MipLevel, parameterName=BrushRT, texture=&TilingNoise05}
  n22 = TextureSampleParameter2D(UVs=Add(A=@n20, B=@n17), Level=@n8) {desc=++X Texel, mipValueMode=TMVM_MipLevel, parameterName=BrushRT, texture=&TilingNoise05}
  n23 = ComponentMask(None=Frac(None=@n1)) {r=true}
  n24 = Divide() {constB=512, desc=Texel Size}
  n25 = Add(A=Floor(None=@n1)) {constB=0.5}
  n26 = Add(A=Multiply(A=@n24, B=Constant2Vector() {g=1}), B=Multiply(A=Divide(A=@n25) {constB=512, desc=/ TextureSize}) {desc=Current Texel Center})
  n27 = Multiply(A=@n24, B=Constant2Vector() {r=1})
  n28 = TextureSampleParameter2D(UVs=Subtract(A=@n26, B=@n27), Level=@n8) {desc=-X Texel, mipValueMode=TMVM_MipLevel, parameterName=BrushRT, texture=&TilingNoise05}
  n29 = TextureSampleParameter2D(UVs=@n26, Level=@n8) {desc="Current Texel (upper left)", mipValueMode=TMVM_MipLevel, parameterName=BrushRT, texture=&TilingNoise05}
  n30 = Add(A=@n27, B=@n26)
  n31 = TextureSampleParameter2D(UVs=@n30, Level=@n8) {desc=+X Texel, mipValueMode=TMVM_MipLevel, parameterName=BrushRT, texture=&TilingNoise05}
  n32 = TextureSampleParameter2D(UVs=Add(A=@n30, B=@n27), Level=@n8) {desc=++X Texel, mipValueMode=TMVM_MipLevel, parameterName=BrushRT, texture=&TilingNoise05}
  n33 = ComponentMask(None=Frac(None=@n1)) {r=true}
  n34 = Divide() {constB=512, desc=Texel Size}
  n35 = Add(A=Floor(None=@n1)) {constB=0.5}
  n36 = Add(A=Multiply(A=@n34, B=Constant2Vector() {g=2}), B=Multiply(A=Divide(A=@n35) {constB=512, desc=/ TextureSize}) {desc=Current Texel Center})
  n37 = Multiply(A=@n34, B=Constant2Vector() {r=1})
  n38 = TextureSampleParameter2D(UVs=Subtract(A=@n36, B=@n37), Level=@n8) {desc=-X Texel, mipValueMode=TMVM_MipLevel, parameterName=BrushRT, texture=&TilingNoise05}
  n39 = TextureSampleParameter2D(UVs=@n36, Level=@n8) {desc="Current Texel (upper left)", mipValueMode=TMVM_MipLevel, parameterName=BrushRT, texture=&TilingNoise05}
  n40 = Add(A=@n37, B=@n36)
  n41 = TextureSampleParameter2D(UVs=@n40, Level=@n8) {desc=+X Texel, mipValueMode=TMVM_MipLevel, parameterName=BrushRT, texture=&TilingNoise05}
  n42 = TextureSampleParameter2D(UVs=Add(A=@n40, B=@n37), Level=@n8) {desc=++X Texel, mipValueMode=TMVM_MipLevel, parameterName=BrushRT, texture=&TilingNoise05}
  FunctionOutput(None=MaterialFunctionCall!(
    t=ComponentMask(None=@n2) {g=true}
    P0=MaterialFunctionCall(t=@n3, P0=@n9.R, P1=@n10.R, P2=@n12.R, P3=@n13.R) {desc=-Y, materialFunction=/Engine/Functions/Engine_MaterialFunctions02/WorldPositionOffset/GenerateASpline}.T position
    P1=MaterialFunctionCall(t=@n14, P0=@n18.R, P1=@n19.R, P2=@n21.R, P3=@n22.R) {desc=Current Y, materialFunction=&GenerateASpline}.T position
    P2=MaterialFunctionCall(t=@n23, P0=@n28.R, P1=@n29.R, P2=@n31.R, P3=@n32.R) {desc=Current Y, materialFunction=&GenerateASpline}.T position
    P3=MaterialFunctionCall(t=@n33, P0=@n38.R, P1=@n39.R, P2=@n41.R, P3=@n42.R) {desc=Current Y, materialFunction=&GenerateASpline}.T position
    Normal/Bi-Normal Cross product vector
   ) {desc=-Y, materialFunction=&GenerateASpline}.T position)
  TextureCoordinate()
 ]
}
