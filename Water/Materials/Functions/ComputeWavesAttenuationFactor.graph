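# Material function that computes a wave attenuation factor from water depth for a given water body, using the Water Wave Data.
# The water body index defaults to a static switch between a scalar parameter and custom HLSL reading it from the tile instance.
# That index and the depth, clamped to zero or above, feed a Custom node calling ComputeWaveDepthAttenuationFactorNew from GerstnerWaveFunctions.ush.
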MaterialFunction /Water/Materials/Functions/ComputeWavesAttenuationFactor
graph {
 properties {description=This function queries the Water Wave Data to get the 'Wave Attenuation Water Depth' value for the current Water Body.}
 nodes [
  MaterialExpressionFunctionInput_3 = FunctionInput(Preview=StaticSwitchParameter(
    True=ScalarParameter() {parameterName=WaterBodyIndex}
    False=Custom() {code="GetWaterWaveParamIndex(Parameters)", description=WaterBodyIndex From Tile Instance, outputType=CMOT_Float1, inputs []}
   ) {parameterName=UseScalarWaterBodyIndex}) {bUsePreviewValueAsDefault=true, inputName=Water Body Index, inputType=FunctionInput_Scalar}
  MaterialExpressionMax_0 = Max(A=FunctionInput() {inputName=WaterDepth, inputType=FunctionInput_Scalar}) {constB=0}
  n1 = Custom!(WaterBodyIndex=@MaterialExpressionFunctionInput_3, WaterDepth=@MaterialExpressionMax_0) {
   code="return ComputeWaveDepthAttenuationFactorNew(WaterBodyIndex, WaterDepth);"
   description=Compute Wave Depth Attenuation Factor
   outputType=CMOT_Float1
   includeFilePaths [/Plugin/Water/Private/GerstnerWaveFunctions.ush]
   inputs [
    {inputName=WaterBodyIndex, input {expression=&self:MaterialExpressionFunctionInput_3}}
    {inputName=WaterDepth, input {expression=&self:MaterialExpressionMax_0}}
   ]
  }
  FunctionOutput(None=@n1) {outputName=WaveAttenuationFactor}
 ]
}
