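# Material function that samples baked water simulation textures at a world XY position, returning velocity with its magnitude, depth and water Z height.
# A named reroute derives UVs from the simulation's double-precision location and size.
# Velocity and depth are decoded from the sim texture using max-value parameters, while height from a second texture is offset by location Z.
# Static switches select virtual textures and cubic or mip-level filtering.
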MaterialFunction /Water/Materials/Functions/SampleBakedFluidSimulation
graph {nodes [
  n1 = TextureObjectParameter() {group=Set From Niagara, parameterName=BakedWaterSimTexVT, samplerType=SAMPLERTYPE_VirtualColor, texture=/Water/Textures/BlackVT4x4}
  n2 = StaticSwitchParameter(True=@n1, False=TextureObjectParameter() {group=Set From Niagara, parameterName=BakedWaterSimTex}) {parameterName=UseVT}
  n3 = StaticBool() {value=true}
  n4 = NamedRerouteUsage()
  n5 = NamedRerouteUsage()
  n6 = FunctionInput(Preview=Constant()) {bUsePreviewValueAsDefault=true, inputName=Mip Level, inputType=FunctionInput_Scalar, sortPriority=3}
  n7 = MaterialFunctionCall(Float4=StaticSwitchParameter(
    True=MaterialFunctionCall!(Texture=@n2, UseSimple=@n3, UVs=@n4) {materialFunction=&/CubicInterpolateTexture}.Result
    False=TextureSample!(UVs=@n5, Tex=@n2, Level=@n6, Apply View MipBias) {mipValueMode=TMVM_MipLevel, samplerSource=SSM_Clamp_WorldGroupSettings}.RGBA
   ) {parameterName=UseHighQualityFiltering}) {materialFunction=&BreakOutFloat4Components}
  n8 = Constant() {r=0.5}
  n9 = Multiply(
   A=Subtract(A=MaterialFunctionCall(X=@n7.G, Y=@n7.G, Z=@n8) {materialFunction=&MakeFloat3}.Result) {constB=0.5}
   B=Multiply(A=ScalarParameter() {parameterName=BakedWaterVelocityMaxSpeed}) {constB=2}
  ) {constB=1212}
  FunctionOutput(None=AppendVector(A=@n9, B=MaterialFunctionCall(Vector 2=@n9) {materialFunction=/Engine/Functions/Engine_MaterialFunctions02/Utility/VectorLength}.V2 Length)) {outputName=Velocity}
  FunctionOutput(None=Multiply(A=@n7.R, B=ScalarParameter() {parameterName=BakedWaterDepthMax})) {outputName=Depth}
  n10 = TextureObjectParameter() {group=Set From Niagara, parameterName=BakedWaterHeightTexVT, samplerType=SAMPLERTYPE_VirtualColor, texture=&BlackVT4x4}
  n11 = StaticSwitchParameter(True=@n10, False=TextureObjectParameter() {group=Set From Niagara, parameterName=BakedWaterHeightTex, texture=&Black}) {parameterName=UseVT}
  n12 = ComponentMask(None=MaterialFunctionCall(Texture=@n11, UseSimple=@n3, UVs=@n4) {materialFunction=&CubicInterpolateTexture}.Result) {r=true}
  n13 = TextureSample(UVs=NamedRerouteUsage(), Tex=@n11) {mipValueMode=TMVM_MipLevel, samplerSource=SSM_Clamp_WorldGroupSettings}
  n14 = DoubleVectorParameter() {parameterName=BakedWaterSimLocationDouble}
  FunctionOutput(None=Add(A=StaticSwitchParameter(True=@n12, False=@n13.R) {parameterName=UseHighQualityFiltering}, B=@n14.Z)) {outputName=Water Z Height}
  n15 = WorldPosition()
  n16 = Subtract(
   A=FunctionInput(Preview=ComponentMask(None=@n15.XYZ) {g=true, r=true}) {inputName=Input Position, inputType=FunctionInput_Vector2}
   B=ComponentMask(None=@n14.XYZ) {g=true, r=true}
  )
  n17 = ComponentMask(None=VectorParameter() {parameterName=BakedWaterSimSize}.RGB) {g=true, r=true}
  NamedRerouteDeclaration(None=Add(A=Divide(A=@n16, B=@n17)) {constB=0.5}) {name=BakedSim_UVs, nodeColor=<rgba 1 0.9141 0 1>}
 ]}
