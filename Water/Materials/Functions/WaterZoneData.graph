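# Material function in the Water library that exposes per-water-zone data to materials.
# A WaterZoneIndex scalar parameter feeds a Custom HLSL node, DecodeWaterZoneData, which calls GetWaterZoneData through an include of the plugin's GerstnerWaveFunctions.ush.
# The node unpacks the returned struct into four function outputs: Location, Extent, HeightExtent and GroundZMin.
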MaterialFunction /Water/Materials/Functions/WaterZoneData
graph {
 properties {libraryCategoriesText [Water]}
 nodes [
  MaterialExpressionScalarParameter_0 = ScalarParameter() {parameterName=WaterZoneIndex}
  n1 = Custom!(WaterZoneIndex=@MaterialExpressionScalarParameter_0) {
   code="FWaterZoneData ZoneData; ZoneData = GetWaterZoneData(WaterZoneIndex);\r\n\r\nExtent = ZoneData.Extent;\r\nHeightExtent = ZoneData.HeightExtent; \r\nGroundZMin= ZoneData.GroundZMin;\r\n\r\nreturn ZoneData.Location;"
   description=DecodeWaterZoneData
   outputType=CMOT_Float2
   additionalOutputs [
    {outputName=Extent, outputType=CMOT_Float2}
    {outputName=HeightExtent, outputType=CMOT_Float2}
    {outputName=GroundZMin, outputType=CMOT_Float1}
   ]
   includeFilePaths [/Plugin/Water/Private/GerstnerWaveFunctions.ush]
   inputs [{inputName=WaterZoneIndex, input {expression=&self:MaterialExpressionScalarParameter_0}}]
  }
  FunctionOutput(None=@n1.return) {outputName=Location}
  FunctionOutput(None=@n1.Extent) {outputName=Extent}
  FunctionOutput(None=@n1.HeightExtent) {outputName=HeightExtent}
  FunctionOutput(None=@n1.GroundZMin) {outputName=GroundZMin}
 ]
}
