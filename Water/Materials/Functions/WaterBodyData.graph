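# Material function that supplies water body velocity, water and terrain Z height, and depth to Water rendering materials.
# The body index, from a scalar parameter or tile instance, selects fixed per-body data and the water zone's info texture slice.
# Static switches choose between fixed values, baked fluid simulation, decoded info-texture samples, or fallbacks like vertex-color flow and zone ground minimum.
# Clamped depth bypasses the fixed-depth switch.
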MaterialFunction /Water/Materials/Functions/WaterBodyData
graph {
 properties {
  bExposeToLibrary=true
  description=This function provides the necessary information about a WaterBody to be used in the Water Rendering materials.
  libraryCategoriesText [Water]
 }
 nodes [
  MaterialExpressionStaticSwitchParameter_8 = StaticSwitchParameter(
   True=ScalarParameter() {parameterName=WaterBodyIndex}
   False=Custom() {code="GetWaterWaveParamIndex(Parameters)", description=WaterBodyIndex From Tile Instance, outputType=CMOT_Float1, inputs []}
  ) {parameterName=UseScalarWaterBodyIndex}
  n1 = Custom!(WaterBodyIndex=@MaterialExpressionStaticSwitchParameter_8) {
   code="FWaterBodyData WaterBodyData = GetWaterBodyData((int)WaterBodyIndex);\r\n\r\nFixedVelocity = WaterBodyData.FixedVelocity;\r\nFixedZHeight = WaterBodyData.FixedZHeight;\r\nFixedWaterDepth = WaterBodyData.FixedWaterDepth;\r\n\r\nreturn 0.0f;"
   description=GetWaterBodyFallbacks
   additionalOutputs [
    {outputName=FixedVelocity, outputType=CMOT_Float3}
    {outputName=FixedZHeight, outputType=CMOT_Float1}
    {outputName=FixedWaterDepth, outputType=CMOT_Float1}
   ]
   includeFilePaths [/Plugin/Water/Private/WaterDataFunctions.ush]
   inputs [{inputName=WaterBodyIndex, input {expression=&self:MaterialExpressionStaticSwitchParameter_8}}]
  }
  n2 = TextureObjectParameter() {parameterName=WaterVelocityAndHeight, samplerType=SAMPLERTYPE_LinearColor, texture=/Water/Textures/BlackTexture2DArray}
  n3 = FunctionInput(Preview=ComponentMask(None=WorldPosition().XYZ) {g=true, r=true}) {bUsePreviewValueAsDefault=true, inputName=Position, inputType=FunctionInput_Vector2}
  n4 = MaterialFunctionCall() {materialFunction=&/WaterZoneData}
  n5 = Subtract(A=@n3, B=Subtract(A=@n4.Location, B=Divide(A=@n4.Extent)))
  MaterialExpressionCustom_6 = Custom(WaterBodyIndex=@MaterialExpressionStaticSwitchParameter_8) {
   code="return GetWaterBodyData((int)WaterBodyIndex).WaterZoneIndex;"
   description=GetWaterZoneIndexForWaterBody
   outputType=CMOT_Float1
   includeFilePaths [&ush]
   inputs [{inputName=WaterBodyIndex, input {expression=&self:MaterialExpressionStaticSwitchParameter_8}}]
  }
  MaterialExpressionReroute_4 = Reroute(None=@MaterialExpressionCustom_6)
  n6 = Custom!(WaterZoneIndex=@MaterialExpressionReroute_4) {
   code="return GetWaterInfoTextureViewIndex(WaterZoneIndex);"
   description=GetWaterInfoTextureViewIndex
   outputType=CMOT_Float1
   showCode=true
   includeFilePaths [&ush]
   inputs [{inputName=WaterZoneIndex, input {expression=&self:MaterialExpressionReroute_4}}]
  }
  MaterialExpressionMaterialFunctionCall_0 = MaterialFunctionCall!(
   Texture=@n2
   UVs=Saturate(None=Divide(A=@n5, B=@n4.Extent))
   Resolution=TextureProperty(Tex=@n2)
   Mip Level=Constant()
   Slice Index=@n6
   Use Cosine UVs
  ) {materialFunction=&/HighQualityTextureFiltering}
  MaterialExpressionScalarParameter_5 = ScalarParameter() {defaultValue=1024, parameterName=MaxFlowVelocity}
  n7 = Custom!(WaterInfoSample=@MaterialExpressionMaterialFunctionCall_0.Result, MaxVelocity=@MaterialExpressionScalarParameter_5) {
   code="return DecodeWaterInfoVelocity(WaterInfoSample, MaxVelocity);"
   description=Decode Flow From Water Info
   outputType=CMOT_Float2
   includeFilePaths [&ush]
   inputs [
    {inputName=WaterInfoSample, input {expression=&self:MaterialExpressionMaterialFunctionCall_0}}
    {inputName=MaxVelocity, input {expression=&self:MaterialExpressionScalarParameter_5}}
   ]
  }
  MaterialExpressionVertexColor_1 = VertexColor()
  n8 = Custom!(Flow=@MaterialExpressionVertexColor_1, MaxVelocity=@MaterialExpressionScalarParameter_5) {
   code="float Magnitude = Flow.x * MaxVelocity;\r\nfloat Direction = Flow.y * 2 * PI;\r\n\r\nfloat FlowX = Magnitude * cos(Direction);\r\nfloat FlowY = Magnitude * sin(Direction);\r\n\r\nreturn float2(FlowX, FlowY);"
   description=Denormalize Flow From Vertex Color
   outputType=CMOT_Float2
   inputs [
    {
     inputName=Flow
     input {expression=&self:MaterialExpressionVertexColor_1, mask=1, maskB=1, maskG=1, maskR=1}
    }
    {inputName=MaxVelocity, input {expression=&self:MaterialExpressionScalarParameter_5}}
   ]
  }
  n9 = StaticSwitchParameter(True=@n7, False=VertexInterpolator(VS=@n8).PS) {defaultValue=true, parameterName=UseWaterInfoTexture}
  n10 = MaterialFunctionCall!(Input Position=@n3, Mip Level) {materialFunction=&/SampleBakedFluidSimulation}
  n11 = ComponentMask(None=@n10.Velocity) {g=true, r=true}
  FunctionOutput(None=StaticSwitchParameter(True=ComponentMask(None=@n1.FixedVelocity) {g=true, r=true}, False=StaticSwitchParameter(True=@n11, False=@n9) {parameterName=UseBakedSim}) {parameterName=UseFixedVelocity}) {outputName=Velocity}
  n12 = Custom!(WaterInfoSample=@MaterialExpressionMaterialFunctionCall_0.Result, WaterZoneIndex=@MaterialExpressionCustom_6) {
   code="return DecodeWaterInfoZHeight(WaterInfoSample, WaterZoneIndex);"
   description=DecodeWaterInfoZHeight
   outputType=CMOT_Float1
   includeFilePaths [&ush]
   inputs [
    {inputName=WaterInfoSample, input {expression=&self:MaterialExpressionMaterialFunctionCall_0}}
    {inputName=WaterZoneIndex, input {expression=&self:MaterialExpressionCustom_6}}
   ]
  }
  n13 = StaticSwitchParameter(True=@n12, False=@n1.FixedZHeight) {defaultValue=true, parameterName=UseWaterInfoTexture}
  n14 = FunctionOutput(None=StaticSwitchParameter(True=@n1.FixedZHeight, False=StaticSwitchParameter(True=@n10.Water Z Height, False=@n13) {parameterName=UseBakedSim}) {parameterName=UseFixedZ}) {outputName=Water Z Height, sortPriority=1}
  n15 = Custom(WaterInfoSample=@MaterialExpressionMaterialFunctionCall_0.Result, WaterZoneIndex=@MaterialExpressionCustom_6) {
   code="return DecodeWaterInfoGroundHeight(WaterInfoSample, WaterZoneIndex);"
   description=DecodeWaterInfoGroundHeight
   outputType=CMOT_Float1
   includeFilePaths [&ush]
   inputs [
    {inputName=WaterInfoSample, input {expression=&self:MaterialExpressionMaterialFunctionCall_0}}
    {inputName=WaterZoneIndex, input {expression=&self:MaterialExpressionCustom_6}}
   ]
  }
  n16 = MaterialFunctionCall() {materialFunction=&WaterZoneData}
  n17 = FunctionOutput(None=StaticSwitchParameter(
    True=Subtract(A=@n10.Water Z Height, B=@n10.Depth)
    False=StaticSwitchParameter(True=@n15, False=@n16.GroundZMin) {defaultValue=true, parameterName=UseWaterInfoTexture}
   ) {parameterName=UseBakedSim}) {outputName=Terrain Z Height, sortPriority=4}
  n18 = StaticSwitchParameter(
   True=@n10.Depth
   False=StaticSwitchParameter(True=Subtract(A=@n14, B=@n17), False=@n1.FixedWaterDepth) {defaultValue=true, parameterName=UseWaterInfoTexture}
  ) {parameterName=UseBakedSim}
  FunctionOutput(None=StaticSwitchParameter(True=@n1.FixedWaterDepth, False=@n18) {parameterName=UseFixedWaterDepth}) {outputName=Water Depth, sortPriority=2}
  FunctionOutput(None=Max(A=@n18) {constB=0}) {outputName=Water Depth Clamped, sortPriority=3}
 ]
}
