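# Water material function, exposed to the library, that fades wave intensity near the shoreline.
# It feeds water depth from WaterBodyData into ComputeWavesAttenuationFactor to derive an attenuation factor, described as using each water body's Wave Attenuation Water Depth.
# That factor scales the incoming WaveAttributes' World Position Offset and blends their Normal from flat (0,0,1) toward the wave normal, returning the modified material attributes.
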MaterialFunction /Water/Materials/Functions/WaveDepthFalloff
graph {
 properties {
  bExposeToLibrary=true
  description="This function fades out wave intensity near the shoeline, using water depth and the per-Waterbody 'Wave Attenuation Water Depth'."
  libraryCategoriesText [Water]
 }
 nodes [
  n1 = GetMaterialAttributes(None=FunctionInput() {inputName=WaveAttributes, inputType=FunctionInput_MaterialAttributes}) {attributeGetTypes [#1, #2]}
  n2 = Constant3Vector() {constant=<rgba 0 0 1 0>}
  n3 = MaterialFunctionCall!(WaterDepth=MaterialFunctionCall!(Position) {materialFunction=&/WaterBodyData}.Water Depth, Water Body Index) {materialFunction=&/ComputeWavesAttenuationFactor}
  FunctionOutput(None=SetMaterialAttributes!(
    MaterialAttributes=@n1.MaterialAttributes
    Normal=LinearInterpolate(A=@n2, B=@n1.Normal, Alpha=@n3.WaveAttenuationFactor)
    WorldPositionOffset=Multiply(A=@n1.WorldPositionOffset, B=@n3.WaveAttenuationFactor)
   ) {attributeSetTypes [#1, #2]})
 ]
}
