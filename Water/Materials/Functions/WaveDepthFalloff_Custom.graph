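# WaveDepthFalloff_Custom is a Water material function that turns water depth into a falloff mask, by name for attenuating waves in shallow water.
# It clamps Depth to non-negative values and divides it by half of Attenuation Depth.
# A Custom HLSL node applies 1-exp(-x), so the mask rises exponentially from zero toward one as depth increases, reaching about 0.86 at the attenuation depth.
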MaterialFunction /Water/Materials/Functions/WaveDepthFalloff_Custom
graph {nodes [
  n1 = FunctionInput() {bUsePreviewValueAsDefault=true, inputName=Attenuation Depth, inputType=FunctionInput_Scalar, previewValue=<100 0 0 1>}
  MaterialExpressionDivide_5 = Divide(A=Max(A=FunctionInput() {bUsePreviewValueAsDefault=true, inputName=Depth, inputType=FunctionInput_Scalar}) {constB=0}, B=Divide(A=@n1))
  n2 = Custom!(x=@MaterialExpressionDivide_5) {
   code="return 1-exp(-x);"
   description=exp
   outputType=CMOT_Float1
   inputs [{inputName=x, input {expression=&self:MaterialExpressionDivide_5}}]
  }
  FunctionOutput(None=@n2) {outputName=Falloff Mask}
 ]}
