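# Experimental material function that boosts the height of water waves near surfaces by scaling the Z of World Position Offset in the incoming material attributes.
# It samples a hard-coded WaterDistanceField texture at the displaced world XY, mapped around the world origin by the MPC_Water capture size.
# A proximity falloff from that distance, multiplied by a wave-height threshold ramp, drives the boost.
# Despite its description, it alters amplitude, not time phase.
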MaterialFunction /Water/Materials/Functions/WaveDistanceFieldReaction
graph {
 properties {
  bExposeToLibrary=true
  description="This is an extremely experimental function that can modify the time phase of waves near surfaces. It is currently referencing a hard coded texture, but should be modified to use Global distance fields in the future."
  libraryCategoriesText [Water]
 }
 nodes [
  n1 = GetMaterialAttributes(None=FunctionInput() {inputType=FunctionInput_MaterialAttributes}) {attributeGetTypes [#1]}
  n2 = ComponentMask(None=@n1.WorldPositionOffset) {b=true}
  n3 = ScalarParameter() {defaultValue=16, parameterName=Min Wave Z For DF Boost}
  n4 = Saturate(None=Divide(A=Subtract(A=@n2, B=@n3), B=ScalarParameter() {defaultValue=8, parameterName=Max Wave Z For DF Boost}))
  n5 = Constant()
  n6 = MaterialFunctionCall(X=@n5, Y=@n5, Z=ScalarParameter() {defaultValue=0.25, parameterName=Wave DF Boost}) {materialFunction=&MakeFloat3}
  n7 = WorldPosition()
  n8 = Add(A=Divide(
    A=Add(A=@n1.WorldPositionOffset, B=@n7.XYZ)
    B=CollectionParameter() {collection=/Water/Materials/MPC/MPC_Water, parameterName=WaveCaptureSize}
   )) {constB=0.5}
  n9 = TextureSample(UVs=ComponentMask(None=Saturate(None=@n8)) {g=true, r=true}) {samplerSource=SSM_Clamp_WorldGroupSettings, samplerType=SAMPLERTYPE_LinearColor, texture=/Water/FluidSimulation/WaveFoam/WaterDistanceField}
  n10 = Add(
   A=Multiply(A=@n9.R, B=CollectionParameter() {collection=&MPC_Water, parameterName=WaveCaptureSize})
   B=ScalarParameter() {defaultValue=64, parameterName=Offset for DF Wave Boost}
  )
  n11 = ScalarParameter() {defaultValue=128, parameterName=Distance from Object for DF boost}
  n12 = OneMinus(None=Saturate(None=Divide(A=@n10, B=@n11)))
  n13 = Add(A=Multiply(A=Multiply(A=@n4, B=@n6.Result), B=@n12))
  FunctionOutput(None=SetMaterialAttributes!(MaterialAttributes=@n1.MaterialAttributes, WorldPositionOffset=Multiply(A=@n1.WorldPositionOffset, B=@n13)) {attributeSetTypes [#1]})
  Multiply(A=@n12, B=ScalarParameter() {parameterName=Use Wave Z TEx})
 ]
}
