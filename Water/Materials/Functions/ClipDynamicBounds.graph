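# Material function that clips water material pixels against the view-dependent dynamic bounds of the water zone.
# It builds XY min and max bounds from WaterZoneData location and half extent, then compares world position.
# UseWaterInfoTexture selects discarding outside or inside those bounds, using Custom HLSL clip calls threaded through BaseColor.
# DisableDynamicBoundsClip passes attributes through untouched for always-rendered materials like Far Mesh or HLOD.
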MaterialFunction /Water/Materials/Functions/ClipDynamicBounds
graph {nodes [
  n1 = GetMaterialAttributes(None=FunctionInput() {inputName=Attributes, inputType=FunctionInput_MaterialAttributes}) {attributeGetTypes [#1]}
  MaterialExpressionReroute_3 = Reroute(None=@n1.BaseColor)
  MaterialExpressionComponentMask_4 = ComponentMask(None=WorldPosition().XYZ) {g=true, r=true}
  n2 = MaterialFunctionCall() {materialFunction=&/WaterZoneData}
  n3 = Divide(A=@n2.Extent)
  MaterialExpressionReroute_1 = Reroute(None=Subtract(A=@n2.Location, B=@n3))
  MaterialExpressionReroute_2 = Reroute(None=Add(A=@n2.Location, B=@n3))
  n4 = Custom!(
   Color=@MaterialExpressionReroute_3
   Position=@MaterialExpressionComponentMask_4
   BoundsMin=@MaterialExpressionReroute_1
   BoundsMax=@MaterialExpressionReroute_2
  ) {
   code="if ((Position.x < BoundsMin.x) || (Position.x > BoundsMax.x) ||\r\n\t(Position.y < BoundsMin.y) || (Position.y > BoundsMax.y))\r\n{\r\n\tclip(-1);\r\n}\r\nreturn Color;"
   description=Clip Outside Bounds
   inputs [
    {inputName=Color, input {expression=&self:MaterialExpressionReroute_3}}
    {inputName=Position, input {expression=&self:MaterialExpressionComponentMask_4}}
    {inputName=BoundsMin, input {expression=&self:MaterialExpressionReroute_1}}
    {inputName=BoundsMax, input {expression=&self:MaterialExpressionReroute_2}}
   ]
  }
  n5 = Custom(
   Color=@MaterialExpressionReroute_3
   Position=@MaterialExpressionComponentMask_4
   BoundsMin=@MaterialExpressionReroute_1
   BoundsMax=@MaterialExpressionReroute_2
  ) {
   code="if ((Position.x > BoundsMin.x) &&&& (Position.x < BoundsMax.x) &&&&\r\n\t(Position.y > BoundsMin.y) &&&& (Position.y < BoundsMax.y))\r\n{\r\n\tclip(-1);\r\n}\r\nreturn Color;"
   description=Clip Inside Bounds
   inputs [
    {inputName=Color, input {expression=&self:MaterialExpressionReroute_3}}
    {inputName=Position, input {expression=&self:MaterialExpressionComponentMask_4}}
    {inputName=BoundsMin, input {expression=&self:MaterialExpressionReroute_1}}
    {inputName=BoundsMax, input {expression=&self:MaterialExpressionReroute_2}}
   ]
  }
  n6 = StaticSwitchParameter(
   True=@n1.MaterialAttributes
   False=SetMaterialAttributes!(
    MaterialAttributes=@n1.MaterialAttributes
    BaseColor=StaticSwitchParameter(True=@n4, False=@n5) {defaultValue=true, parameterName=UseWaterInfoTexture}
   ) {attributeSetTypes [#1]}
  ) {
   desc="Disables clipping when this material is rendering outside of the active dynamic bounds. The dynamic bounds are view dependent and some materials like the Far Mesh or HLOD don't need to care about this region, as they should always render."
   parameterName=DisableDynamicBoundsClip
  }
  FunctionOutput(None=@n6) {outputName=Attributes}
 ]}
