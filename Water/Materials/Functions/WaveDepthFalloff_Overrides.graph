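# Water material function that fades wave intensity near shorelines, exposing overridable inputs for position and water body index.
# It samples water depth at the position (default world XY) and feeds it with the body index (default from the tile instance) into ComputeWavesAttenuationFactor.
# The resulting factor scales the incoming wave attributes' World Position Offset and lerps their Normal from straight up.
# It outputs both the adjusted attributes and the Attenuation Factor.
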MaterialFunction /FortWater/Materials/Functions/WaveDepthFalloff_Overrides
graph {
 properties {
  bExposeToLibrary=true
  description="This function fades out wave intensity near the shoeline, using water depth and the per-Waterbody 'Wave Attenuation Water Depth'. This _Overrides version has more inputs to override vs the base version."
  libraryCategoriesText [Water]
 }
 nodes [
  n1 = GetMaterialAttributes(None=FunctionInput() {inputName=WaveAttributes, inputType=FunctionInput_MaterialAttributes, sortPriority=2}) {attributeGetTypes [#1, #2]}
  n2 = Constant3Vector() {constant=<rgba 0 0 1 0>}
  n3 = FunctionInput(Preview=ComponentMask(None=WorldPosition().XYZ) {g=true, r=true}) {bUsePreviewValueAsDefault=true, inputName=Position, inputType=FunctionInput_Vector2, sortPriority=1}
  n4 = Custom() {code="GetWaterWaveParamIndex(Parameters)", description=WaterBodyIndex From Tile Instance, outputType=CMOT_Float1, inputs []}
  n5 = FunctionOutput(None=MaterialFunctionCall!(
    WaterDepth=MaterialFunctionCall!(Texture, Position=@n3) {materialFunction=&/WaterBodyData}.Water Depth
    Water Body Index=FunctionInput(Preview=@n4) {bUsePreviewValueAsDefault=true, inputName=Water Body Index, inputType=FunctionInput_Scalar}
   ) {materialFunction=/Water/Materials/Functions/ComputeWavesAttenuationFactor}.WaveAttenuationFactor) {outputName=Attenuation Factor, sortPriority=1}
  FunctionOutput(None=SetMaterialAttributes!(
    MaterialAttributes=@n1.MaterialAttributes
    Normal=LinearInterpolate(A=@n2, B=@n1.Normal, Alpha=@n5)
    WorldPositionOffset=Multiply(A=@n1.WorldPositionOffset, B=@n5)
   ) {attributeSetTypes [#1, #2]})
 ]
}
