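# Material function that computes Gerstner wave normal and world position offset for WaterBody materials.
# Custom HLSL calls GerstnerWaveFunctions.ush with water body index, position and time, returning both results packed in one value.
# Two custom nodes unpack the normal from the fraction and the offset from the integer part.
# A default-on static switch applies FortWater depth falloff before outputting Normal, WPO and combined attributes.
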MaterialFunction /Water/Materials/Functions/ComputeGerstnerWaves
graph {
 properties {
  bExposeToLibrary=true
  description=This function computes the Gerstner Waves for a Water material that comes from a WaterBody actor.
  libraryCategoriesText [Water]
 }
 nodes [
  MaterialExpressionFunctionInput_3 = FunctionInput(Preview=Custom() {code="GetWaterWaveParamIndex(Parameters)", description=WaterBodyIndex From Tile Instance, outputType=CMOT_Float1, inputs []}) {bUsePreviewValueAsDefault=true, inputName=Water Body Index, inputType=FunctionInput_Scalar}
  MaterialExpressionFunctionInput_1 = FunctionInput(Preview=WorldPosition() {worldPositionShaderOffset=WPT_ExcludeAllShaderOffsets}.XYZ) {bUsePreviewValueAsDefault=true, inputName=Override Position, inputType=FunctionInput_Vector2, sortPriority=1}
  MaterialExpressionFunctionInput_2 = FunctionInput() {inputName=Time, inputType=FunctionInput_Scalar, sortPriority=2}
  MaterialExpressionCustom_0 = Custom!(WaterBodyIndex=@MaterialExpressionFunctionInput_3, WorldPosition=@MaterialExpressionFunctionInput_1, Time=@MaterialExpressionFunctionInput_2) {
   code="WaveOutput Waves;\r\n\r\nWaves = GetAllGerstnerWavesNew(WaterBodyIndex, Time, LWCWorldPosition);\r\nreturn PackNormalAndWPONew(Waves);"
   description=Get Gerstner Waves
   includeFilePaths [/Plugin/Water/Private/GerstnerWaveFunctions.ush]
   inputs [
    {inputName=WaterBodyIndex, input {expression=&self:MaterialExpressionFunctionInput_3}}
    {inputName=WorldPosition, input {expression=&self:MaterialExpressionFunctionInput_1}}
    {inputName=Time, input {expression=&self:MaterialExpressionFunctionInput_2}}
   ]
  }
  n1 = Custom!(inPackedWave=@MaterialExpressionCustom_0) {
   code="\t\tfloat3 outnormal = frac(inPackedWave);\r\n\t\toutnormal *= 2.02;\r\n\t\toutnormal -= 1.01;\r\n\t\treturn outnormal;"
   description=UnpackNormal
   inputs [{inputName=inPackedWave, input {expression=&self:MaterialExpressionCustom_0}}]
  }
  n2 = Custom(inPackedWave=@MaterialExpressionCustom_0) {
   code="\t\tfloat3 outWPO;\r\n\t\toutWPO = floor(inPackedWave) / 100;\r\n\t\treturn outWPO;"
   description=UnpackWPO
   inputs [{inputName=inPackedWave, input {expression=&self:MaterialExpressionCustom_0}}]
  }
  n3 = SetMaterialAttributes!(MaterialAttributes, Normal=@n1, WorldPositionOffset=@n2) {attributeSetTypes [#1, #2]}
  n4 = StaticBool() {value=true}
  n5 = GetMaterialAttributes(None=StaticSwitch(
    True=MaterialFunctionCall!(Water Body Index=@MaterialExpressionFunctionInput_3, Position=@MaterialExpressionFunctionInput_1, WaveAttributes=@n3) {materialFunction=/FortWater/Materials/Functions/WaveDepthFalloff_Overrides}.Result
    False=@n3
    Value=FunctionInput(Preview=@n4) {bUsePreviewValueAsDefault=true, inputName=Use Depth Attenuation, inputType=FunctionInput_StaticBool, sortPriority=5}
   )) {attributeGetTypes [#1, #2]}
  FunctionOutput(None=@n5.Normal) {outputName=Normal, sortPriority=1}
  FunctionOutput(None=@n5.WorldPositionOffset) {outputName=WPO, sortPriority=2}
  FunctionOutput(None=@n5.MaterialAttributes) {outputName=Wave Attributes}
 ]
}
