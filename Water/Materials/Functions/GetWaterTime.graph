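# GetWaterTime is a library material function that supplies one shared time value so water materials stay in sync.
# It reads Time and PrevTime from MPC_Water through a PreviousFrameSwitch, keeping previous-frame evaluation correct.
# A lerp driven by the FreezeTime parameter blends that value toward ManualTime, which lets r.water.freezewaves freeze all water.
# The result is exposed as the WaterTime output.
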MaterialFunction /Water/Materials/Functions/GetWaterTime
graph {
 properties {
  bExposeToLibrary=true
  description=This function helps sync Time across various places in Water. It allows the following cvar to be adjusted to freeze time for all water: r.water.freezewaves 1.
  libraryCategoriesText [Water]
 }
 nodes [
  n1 = CollectionParameter() {collection=/Water/Materials/MPC/MPC_Water, parameterName=Time}
  n2 = CollectionParameter() {collection=&MPC_Water, parameterName=PrevTime}
  FunctionOutput(None=LinearInterpolate(
    A=PreviousFrameSwitch(Current Frame=@n1, Previous Frame=@n2)
    B=CollectionParameter() {collection=&MPC_Water, parameterName=ManualTime}
    Alpha=CollectionParameter() {collection=&MPC_Water, parameterName=FreezeTime}
   )) {outputName=WaterTime}
 ]
}
