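# Water material function, exposed to the library, that merges a wave attributes input into base material attributes, keeping wave rendering separate from standard shading.
# It extracts attributes from the wave input and sets Normal and WorldPositionOffset on the base, though both are wired from the extracted Normal output.
# Unconnected inputs default to the WaterAttributes function for the base and a flat normal with zero offset for waves.
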MaterialFunction /Water/Materials/Functions/SetWaveAttributes
graph {
 properties {
  bExposeToLibrary=true
  description=WaveAttributes is just a loose term to decsribe a MaterialAttributes struct that has data for Normal and WPO set. This is a convenient way to separate the standard shading path from the Wave rendering path.
  libraryCategoriesText [Water]
 }
 nodes [
  n1 = MaterialFunctionCall() {materialFunction=/Water/Materials/Layers/WaterAttributes}
  n2 = Constant3Vector() {constant=<rgba 0 0 1 0>}
  n3 = Constant()
  n4 = GetMaterialAttributes(None=FunctionInput(Preview=SetMaterialAttributes!(MaterialAttributes, Normal=@n2, WorldPositionOffset=@n3) {attributeSetTypes [#1, #2]}) {bUsePreviewValueAsDefault=true, inputName=Wave Attributes, inputType=FunctionInput_MaterialAttributes, sortPriority=1}) {attributeGetTypes [#1, #2]}
  FunctionOutput(None=SetMaterialAttributes(
    MaterialAttributes=FunctionInput(Preview=@n1.Result) {bUsePreviewValueAsDefault=true, inputName=Material Atttributes, inputType=FunctionInput_MaterialAttributes}
    Normal=@n4.Normal
    WorldPositionOffset=@n4.Normal
   ) {attributeSetTypes [#1, #2]}) {outputName=MA}
 ]
}
