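# Material function in the Water library that masks SingleLayerWater Scattering and Absorption coefficients, which is needed to make that water see-through.
# A scalar Mask input multiplies the Scattering and Absorption inputs separately, and each product goes to a matching output.
# The Mask also passes through unchanged as a third output.
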MaterialFunction /Water/Materials/Functions/WaterCoefficientMask
graph {
 properties {
  bExposeToLibrary=true
  description=This function provides a convenient way to mask the SingleLayerWater Scattering and Absorption coefficients. This is necessary to make SingleLayerWater see-through.
  libraryCategoriesText [Water]
 }
 nodes [
  n1 = FunctionInput() {bUsePreviewValueAsDefault=true, inputName=Mask, inputType=FunctionInput_Scalar, previewValue=<1 0 0 1>}
  FunctionOutput(None=Multiply(A=@n1, B=FunctionInput() {bUsePreviewValueAsDefault=true, inputName=Scattering, sortPriority=1})) {outputName=Scattering, sortPriority=1}
  FunctionOutput(None=Multiply(A=@n1, B=FunctionInput() {bUsePreviewValueAsDefault=true, inputName=Absorption, sortPriority=2})) {outputName=Absorption, sortPriority=2}
  FunctionOutput(None=@n1) {outputName=Mask}
 ]
}
