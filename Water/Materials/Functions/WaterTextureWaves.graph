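# Material function that creates complex wave motion by panning one input texture in diagonal directions with offset, rescaled coordinates, described as 4-way diagonal motion.
# A shared speed-scaled time drives four panners, with irregular offsets and golden-ratio-like divisors breaking up repetition.
# Only three samples reach the output, summed and multiplied by Divisor (default one third).
# The fourth sample feeds an Add left unconnected to the output.
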MaterialFunction /Water/Materials/Functions/WaterTextureWaves
graph {
 properties {description=4-way diagonal motion, libraryCategoriesText [Misc, ComplexMotion]}
 nodes [
  n1 = FunctionInput(Preview=TextureCoordinate()) {bUsePreviewValueAsDefault=true, inputName=Coordinates, inputType=FunctionInput_Vector2}
  n2 = Add(A=@n1, B=Constant2Vector() {g=0.1484, r=0.8649})
  n3 = FunctionInput() {bUsePreviewValueAsDefault=true, inputName=Speed, inputType=FunctionInput_Scalar, previewValue=<1 0 0 1>}
  n4 = MaterialFunctionCall(Time=FunctionInput(Preview=Time()) {bUsePreviewValueAsDefault=true, inputName=Time, inputType=FunctionInput_Scalar}, Speed=@n3) {materialFunction=&TimeWithSpeedVariable}
  n5 = FunctionInput(Preview=TextureObject()) {inputName=Texture, inputType=FunctionInput_Texture2D, sortPriority=1}
  n6 = TextureSample(UVs=Panner(Coordinate=Divide(A=@n2) {constB=1.2361}, Time=@n4.Time) {speedX=-0.1, speedY=0.1}, Tex=@n5) {samplerSource=SSM_Wrap_WorldGroupSettings}
  n7 = Divide(A=Add(A=@n1, B=Constant2Vector() {g=0.7516, r=0.6513})) {constB=1.8541}
  Add(
   A=@n6.RGB
   B=TextureSample(UVs=Panner(Coordinate=@n7, Time=@n4.Time) {speedX=0.1, speedY=-0.1}, Tex=@n5) {samplerSource=SSM_Wrap_WorldGroupSettings}.RGB
  )
  n8 = Divide(A=Add(A=@n1, B=Constant2Vector() {g=0.3548, r=0.4181})) {constB=1.618}
  n9 = Add(
   A=TextureSample(UVs=Panner(Coordinate=@n1, Time=@n4.Time) {speedX=0.1, speedY=0.1}, Tex=@n5) {samplerSource=SSM_Wrap_WorldGroupSettings}.RGB
   B=TextureSample(UVs=Panner(Coordinate=@n8, Time=@n4.Time) {speedX=-0.1, speedY=-0.1}, Tex=@n5) {samplerSource=SSM_Wrap_WorldGroupSettings}.RGB
  )
  n10 = FunctionInput() {
   bUsePreviewValueAsDefault=true
   description=How much to reduce the overall combined results of the 4 way add.
   inputName=Divisor
   inputType=FunctionInput_Scalar
   previewValue=<0.3333 0 0 1>
  }
  FunctionOutput(None=Multiply(A=Add(A=@n9, B=@n6.RGB), B=@n10))
 ]
}
