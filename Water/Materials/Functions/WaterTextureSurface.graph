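# Material function that layers tiling texture detail onto incoming water wave attributes, adjusting Normal and World Position Offset by pixel depth.
# WaterTextureWaves samples a normal-with-height texture in world XY, with the scale switching hard from near to distant via Ceil.
# Near detail adds to the wave normal and fades with depth.
# Farther out, a texture-only normal with Fresnel-driven default strength and height-based vertical displacement progressively replace the wave values.
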MaterialFunction /Water/Materials/Functions/WaterTextureSurface
graph {nodes [
  FunctionInput() {bUsePreviewValueAsDefault=true, inputName=Blend Alpha, inputType=FunctionInput_Scalar, sortPriority=2, previewValue=<1 0 0 1>}
  n1 = GetMaterialAttributes(None=FunctionInput(Preview=SetMaterialAttributes!(MaterialAttributes)) {bUsePreviewValueAsDefault=true, inputName=Wave Attributes, inputType=FunctionInput_MaterialAttributes}) {attributeGetTypes [#1, #2]}
  n2 = WorldPosition() {worldPositionShaderOffset=WPT_ExcludeAllShaderOffsets}
  n3 = FunctionInput(Preview=ScalarParameter() {defaultValue=512, group=Detail Normal Settings, parameterName=Default Near Water Scale}) {bUsePreviewValueAsDefault=true, inputName=Near Scale, inputType=FunctionInput_Scalar, sortPriority=2, previewValue=<2048 0 0 1>}
  n4 = FunctionInput(Preview=ScalarParameter() {defaultValue=3500, group=Detail Normal Settings, parameterName=Default Distant Water Scale}) {bUsePreviewValueAsDefault=true, inputName=Distant Scale, inputType=FunctionInput_Scalar, sortPriority=1, previewValue=<2048 0 0 1>}
  n5 = PixelDepth()
  n6 = ScalarParameter() {defaultValue=5000, group=Detail Normal Settings, parameterName=Water Distant Normal Offset}
  n7 = Saturate(None=Divide(
    A=Subtract(A=@n5.R, B=@n6)
    B=ScalarParameter() {defaultValue=10000, group=Detail Normal Settings, parameterName=Water Distant Normal Length}
   ))
  n8 = LinearInterpolate(A=@n3, B=@n4, Alpha=Saturate(None=Ceil(None=@n7)))
  n9 = FunctionInput(Preview=ScalarParameter() {defaultValue=0.75, group=Detail Normal Settings, parameterName=Default Disant Water Speed}) {bUsePreviewValueAsDefault=true, inputName=Speed, inputType=FunctionInput_Scalar, sortPriority=3, previewValue=<1 0 0 1>}
  n10 = MaterialFunctionCall!(
   Coordinates=ComponentMask(None=Divide(A=@n2.XYZ, B=@n8)) {g=true, r=true}
   Divisor
   Speed=@n9
   Time=FunctionInput(Preview=Time()) {bUsePreviewValueAsDefault=true, inputName=Time, inputType=FunctionInput_Scalar}
   Texture=TextureObject() {samplerType=SAMPLERTYPE_LinearColor, texture=/Water/Textures/Normals/T_Water_TilingNormal_With_Height_02_Softened}
  ) {materialFunction=&/WaterTextureWaves}
  n11 = Subtract(A=ComponentMask(None=@n10.Result) {g=true, r=true}) {constB=0.5}
  n12 = FunctionInput(Preview=ScalarParameter() {defaultValue=0.25, group=Detail Normal Settings, parameterName=Default Near Normal Strength}) {bUsePreviewValueAsDefault=true, inputName=Near Normal Strength, inputType=FunctionInput_Scalar, sortPriority=5, previewValue=<1 0 0 1>}
  n13 = PixelDepth()
  n14 = ScalarParameter() {defaultValue=5000, group=Detail Normal Settings, parameterName=Water Distant Normal Offset}
  n15 = LinearInterpolate(A=@n12, B=Constant(), Alpha=Saturate(None=Divide(A=@n13.R, B=@n14))) {constB=0}
  n16 = Add(A=@n1.Normal, B=AppendVector(A=Multiply(A=@n11, B=@n15), B=Constant()))
  n17 = ScalarParameter() {defaultValue=9, group=Detail Normal Settings, parameterName=Far Normal Fresnel Power}
  n18 = Constant3Vector() {constant=<rgba 0 0 1 0>}
  n19 = FunctionInput(Preview=LinearInterpolate(
    A=ScalarParameter() {defaultValue=0.25, group=Detail Normal Settings, parameterName=Default Distant Normal StrengthB}
    B=ScalarParameter() {defaultValue=0.3, group=Detail Normal Settings, parameterName=Default Distant Normal Strength}
    Alpha=Fresnel(ExponentIn=@n17, Normal=@n18)
   )) {bUsePreviewValueAsDefault=true, inputName=Far Normal Strength, inputType=FunctionInput_Scalar, sortPriority=4, previewValue=<1 0 0 1>}
  n20 = Normalize(VectorInput=AppendVector(A=@n11, B=Divide(B=@n19)))
  n21 = Subtract(A=ComponentMask(None=@n10.Result) {b=true}) {constB=0.5}
  n22 = Multiply(
   A=Multiply(
    A=Multiply(A=@n21, B=@n7)
    B=FunctionInput() {bUsePreviewValueAsDefault=true, inputName=Far Displacement, inputType=FunctionInput_Scalar, sortPriority=6}
   )
   B=Constant3Vector() {constant=<rgba 0 0 1 0>}
  )
  FunctionOutput(None=SetMaterialAttributes!(
    MaterialAttributes=@n1.MaterialAttributes
    Normal=LinearInterpolate(A=@n16, B=@n20, Alpha=@n7)
    WorldPositionOffset=LinearInterpolate(A=@n1.WorldPositionOffset, B=@n22, Alpha=@n7)
   ) {attributeSetTypes [#1, #2]}) {description=Material Attributes with Normal and WPO defined., outputName=Wave Attributes}
  Reroute(None=@n7)
 ]}
