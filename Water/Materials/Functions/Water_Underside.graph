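# Material function that adapts water materials for viewing the underside of the surface, using TwoSidedSign to separate front and back faces.
# On back faces, a normal-camera dot product against a critical angle drives Specular and Opacity to one, mimicking total internal reflection.
# Back-face Refraction blends between a scaled input value and a lip value near that angle.
# Separate Scattering and Absorption inputs are zeroed on back faces.
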MaterialFunction /Water/Materials/Functions/Water_Underside
graph {
 properties {
  bExposeToLibrary=true
  description="Adjusts Specular, Opacity and Refraction to allow Water materials to work for the bottom side of the water using TwoSidedSign."
  libraryCategoriesText [Water]
 }
 nodes [
  n1 = FunctionInput() {inputName=Attributes, inputType=FunctionInput_MaterialAttributes}
  n2 = GetMaterialAttributes(None=@n1) {attributeGetTypes [#1, #2, #3]}
  n3 = Subtract(
   A=ScalarParameter() {defaultValue=0.225, group=TwoSided Settings, parameterName=Critical Angle Dot}
   B=DotProduct(A=PixelNormalWS(), B=CameraVectorWS())
  )
  n4 = ScalarParameter() {defaultValue=0.01, group=TwoSided Settings, parameterName=Critical Angle Width}
  n5 = LinearInterpolate(A=Saturate(None=Divide(A=@n3, B=@n4)), Alpha=Saturate(None=TwoSidedSign())) {constB=0}
  n6 = Divide(A=@n3, B=ScalarParameter() {defaultValue=-0.75, group=TwoSided Settings, parameterName=Critical Angle Bend Width})
  n7 = LinearInterpolate(
   A=LinearInterpolate(B=@n2.Refraction, Alpha=ScalarParameter() {defaultValue=1, group=TwoSided Settings, parameterName=Refraction Bottom Amount}) {constA=1}
   B=ScalarParameter() {defaultValue=0.95, group=TwoSided Settings, parameterName=Lip Refraction}
   Alpha=OneMinus(None=Saturate(None=@n6))
  ) {constA=1}
  n8 = Saturate(None=TwoSidedSign())
  FunctionOutput(None=SetMaterialAttributes!(
    MaterialAttributes=@n1
    Specular=LinearInterpolate(A=@n2.Specular, Alpha=@n5)
    Opacity=LinearInterpolate(A=@n2.Opacity, Alpha=@n5)
    Refraction=LinearInterpolate(A=@n7, B=@n2.Refraction, Alpha=@n8)
   ) {attributeSetTypes [#1, #2, #3]})
  n9 = Saturate(None=TwoSidedSign())
  FunctionOutput(None=Multiply(A=FunctionInput() {inputName=Scattering, sortPriority=1}, B=@n9)) {outputName=Scattering, sortPriority=1}
  FunctionOutput(None=Multiply(A=FunctionInput() {inputName=Absorption, sortPriority=2}, B=@n9)) {outputName=Absorption, sortPriority=2}
 ]
}
