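# Material function that reconstructs the world position of the scene behind SingleLayerWater surfaces, exposed to the library under Water.
# An optional ScreenUV input, defaulting to the pixel's viewport UV, offsets the negated camera vector through a field-of-view-scaled view-to-world vector.
# SceneDepthWithoutWater at that UV is divided by the ray's alignment with camera forward, then projected from the camera position.
# Reconstruction quality depends on r.Water.SingleLayer.RefractionDownsampleFactor.
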MaterialFunction /Water/Materials/Functions/WorldPositionBehindWater
graph {
 properties {
  bExposeToLibrary=true
  description=Constructs the WorldPosition of the scene behind SingleLayerWater materials. Only compatible in SingleLayerWater materials. The quality of the construction is based on r.Water.SingleLayer.RefractionDownsampleFactor.
  libraryCategoriesText [Water]
 }
 nodes [
  n1 = ScreenPosition()
  n2 = FunctionInput(Preview=@n1.ViewportUV) {bUsePreviewValueAsDefault=true, inputName=ScreenUV, inputType=FunctionInput_Vector2}
  n3 = MaterialFunctionCall() {materialFunction=&CameraDirectionVector}
  n4 = CameraVectorWS()
  n5 = Subtract(A=@n2, B=@n1.ViewportUV)
  n6 = Constant2Vector() {g=-2, r=2}
  n7 = ComponentMask(None=ViewProperty() {property=MEVP_TanHalfFieldOfView}.Property) {g=true, r=true}
  n8 = AppendVector(A=Multiply(A=Multiply(A=@n5, B=@n6) {constB=2}, B=Multiply(A=@n7) {constB=2}), B=Constant())
  n9 = Normalize(VectorInput=Add(A=Multiply(A=@n4) {constB=-1}, B=Transform(None=@n8) {transformSourceType=TRANSFORMSOURCE_View}))
  n10 = Divide(A=SceneDepthWithoutWater!(UVs=@n2).R, B=Abs(None=DotProduct(A=@n3.Result, B=@n9)))
  FunctionOutput(None=Add(A=CameraPositionWS(), B=Multiply(A=@n10, B=@n9)))
 ]
}
